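# Superlevel skin material for the LazyLizz Chip character's face accessory, recoloring it and adding a glowing logo.
# It fully desaturates the PreFX parent function's base color, then multiplies it by mask-driven tints: red for a fresnel-modulated primary color, blue for secondary, green for a dark logo base.
# The green mask also drives emissive logo color.
# The modified attributes then pass through the shared PostFX character function.
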Material /BRCosmetics/Characters/Player/Male/Medium/Bodies/M_MED_LazyLizz_Chip/Skins/Superlevel/M_LazyLizz_Chip_FaceAcc_SL
graph {
 properties {
  bUseMaterialAttributes=true
  bUsedWithMeshDeformer=true
  bUsedWithMorphTargets=true
  bUsedWithSkeletalMesh=true
  shadingModel=MSM_FromMaterialExpression
  usedShadingModels=Default Lit | Subsurface | Clear Coat | Subsurface Profile | Cloth
 }
 nodes [
  n1 = MaterialFunctionCall() {materialFunction=&MF_CharacterParent_2023_PreFX}
  n2 = GetMaterialAttributes(None=@n1.Result) {attributeGetTypes [#1]}
  n3 = Desaturation(None=@n2.Base Color, Fraction=Constant() {r=1})
  n4 = TextureSample() {
   samplerSource=SSM_Wrap_WorldGroupSettings
   samplerType=SAMPLERTYPE_Masks
   texture=/BRCosmetics/Characters/Player/Male/Medium/Bodies/M_MED_LazyLizz_Chip/Textures/T_LazyLizz_Chip_FaceAccExtra_SL_FX
  }
  n5 = NamedRerouteUsage()
  n6 = ScalarParameter() {defaultValue=0.2, group=Superlevel, parameterName=FresnelOffset}
  n7 = Multiply(
   A=Saturate(None=Add(A=@n5, B=@n6) {constB=0.15})
   B=VectorParameter() {group=Superlevel, parameterName=PrimaryColor, defaultValue=<rgba 0.6817 0.1147 0.6858 1>}.RGB
  )
  n8 = VectorParameter() {group=Superlevel, parameterName=SecondaryColor, defaultValue=<rgba 0.3319 0.0034 1 1>}
  n9 = Add(A=Add(A=Multiply(A=@n4.R, B=@n7), B=Multiply(A=@n4.B, B=@n8.RGB)), B=Multiply(A=@n4.G) {constB=0.05})
  n10 = VectorParameter() {group=Superlevel, parameterName=EmissiveLogoColor, defaultValue=<rgba 3 0.0915 2.6467 1>}
  MaterialFunctionCall(
   Material=SetMaterialAttributes!(MaterialAttributes=@n2.MaterialAttributes, Base Color=Multiply(A=@n3, B=@n9), Emissive Color=Multiply(A=@n4.G, B=@n10.RGB)) {attributeSetTypes [#1, #2]}
   M=@n1.M
   Pre Skinned Local Position=@n1.M
   Distance Blend=@n1.M
   Fresnel=@n1.M
  ) {materialFunction=&MF_CharacterParent_2023_PostFX}
  NamedRerouteDeclaration(None=@n1.Fresnel) {name=InFresnel, nodeColor=<rgba 0.5234 0 1 1>}
 ]
}
